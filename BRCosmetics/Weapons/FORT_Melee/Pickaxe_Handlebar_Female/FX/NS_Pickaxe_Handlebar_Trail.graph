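# Niagara system for the swing trail of the female Handlebar pickaxe cosmetic, drawn as faint, short-lived pale beige ribbons.
# One emitter spawns two anim trails through the FN_ anim trail modules, each trail initialized with its own bone or socket index selection.
# The update step follows the User.SkeletalMesh data interface, and a ribbon renderer draws the result.
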NiagaraSystem /BRCosmetics/Weapons/FORT_Melee/Pickaxe_Handlebar_Female/FX/NS_Pickaxe_Handlebar_Trail
dependencies {usedDataInterfaces [], usedDynamicInputs [], usedModules [&SystemState]}
emitters {FN_AnimTrail_Test {
  emitterUpdateScript [
   &EmitterState()
   &FN_SpawnAnimTrail(Number of Anim Trails=2)
  ]
  particleSpawnScript [
   &FN_InitializeAnimTrailBoneIndecies()
   &FN_InitializeAnimTrailBoneIndecies(AnimTrail Selection=1, First Bone or Socket Index=2, Second Bone or Socket Index=3) {module=FN_InitializeAnimTrailBoneIndecies001}
   &FN_InitializeAnimTrail(Color=<rgba 0.6 0.58 0.48 0.2>, Lifetime=0.15)
  ]
  particleUpdateScript [
   &SolveForcesAndVelocity(Clamp Velocity=false, Limit Acceleration=false)
   &FN_UpdateAnimTrail(Locked Particle RibbonWidth=1, SkeletalMesh=$User.SkeletalMesh) {module=FN_UpdateAnimTrail001}
  ]
  rendererClasses [&NiagaraRibbonRendererProperties]
  renderers [{rendererClass=&NiagaraRibbonRendererProperties}]
 }}
summary {
 systemName=NS_Pickaxe_Handlebar_Trail
 emitters [{emitterName=FN_AnimTrail_Test, rendererClasses [&NiagaraRibbonRendererProperties]}]
 userVariables [
  User.RibbonIDOffset {type=NiagaraInt32}
  User.RibbonWidth {type=NiagaraFloat, defaultValue=10, description=""}
  User.SkeletalMesh {
   type=NiagaraDataInterfaceSkeletalMesh
   defaultValue {
    struct=NiagaraExt_VariableValue_DataInterface
    value {dataInterface=&self:NiagaraDataInterfaceSkeletalMesh_0}
   }
  }
 ]
}
