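# Physics asset for the Handlebar Female pickaxe melee weapon mesh in the BR cosmetics content.
# It holds a single capsule body on the Melee_TwoHanded_Handle bone, offset and rotated to run along the weapon.
# It defines no constraints, so the weapon simulates and collides as one rigid shape.
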PhysicsAsset /BRCosmetics/Weapons/FORT_Melee/Pickaxe_Handlebar_Female/Meshes/Handlebar_Female_Axe_PhysicsAsset
bodies {Melee_TwoHanded_Handle [{shapeType=Capsule, length=71.541, radius=10.2199, center=<1.0883 -0.1442 28.5972>, rotation=<rot 4.4979 -90 0.03>}]}
constraints []
